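# Niagara system for the VoidSpray backpack that flashes a flickering holographic score from the system kill count.
# Emissive_Heart001 derives two digit values from that count and animates glitchy offsets and stretch at a mesh socket.
# Illumination places three beam, flare and laser elements on separate sockets, blending alignment and size by FocusFactor.
# Circle and square emitters add glows and flickering debris around the display socket, with shared FadeIn, ReDelay and GlobalFlicker values driving timing and alpha.
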NiagaraSystem /BRCosmetics/Accessories/FORT_Backpacks/Backpack_F_MED_VoidSpray/FX/NS_Backpack_VoidSpray_Score
dependencies {
 usedDataInterfaces [
  &NiagaraDataInterfaceCurve
  &NiagaraDataInterfaceColorCurve
  &NiagaraDataInterfaceSpriteRendererInfo
  &NiagaraDataInterfaceSocketReader
 ]
 usedDynamicInputs [&MakeCustomIntFromBool]
 usedModules [
  &self:SystemSpawnScript.SystemScriptSource.SystemScriptGraph.NiagaraNodeAssignment_0.SetVariables_#1
  &SystemState
  &self:SystemSpawnScript.SystemScriptSource.SystemScriptGraph.NiagaraNodeAssignment_3.SetVariables_#2
  &self:SystemSpawnScript.SystemScriptSource.SystemScriptGraph.NiagaraNodeAssignment_6.SetVariables_#3
  &self:SystemSpawnScript.SystemScriptSource.SystemScriptGraph.NiagaraNodeAssignment_1.SetVariables_#4
 ]
}
emitters {
 Emissive_Heart001 {
  emitterUpdateScript [
   &EmitterState(Life Cycle Mode=::NewEnumerator1, Loop Behavior=::NewEnumerator1, Loop Delay=0.5, Loop Duration=0.3, UseLoopDelay=true)
   &SpawnBurst_Instantaneous(Spawn Time=0.01, Use Spawn Probability=false, UseLoopCountLimit=true)
  ]
  particleSpawnScript [
   &InitializeParticle(
    Lifetime=2.5
    Sprite Size=&Multiply_Vector2D_ByFloat(Float=20.2961, Vector2D=<0.4 0.8>)
    Sprite Size Mode=::NewEnumerator4
    UsePositionOffset=false
    Write Material Random=true
   )
   &CameraOffset(Camera Offset Amount=10)
   &SocketLocation(SamplePosOffset=false, Socket Index=3, Socket Reader=$System.Socket Reader, Socket Sampling Mode=::NewEnumerator3, VelocitySpeedLimit=false)
  ]
  particleUpdateScript [
   &ParticleState()
   &ScaleColor(
    Linear Color Curve=DataInterface() {propertyValues {
      AlphaCurve {keys [<at 0 0.5>,<at 0.0454 2>,<at 0.1047 1>,<at 0.7513 1>,<at 0.8856 0.691>,<at 0.9146 1.5>,<at 1 0>]}
      BlueCurve {keys [<at 0 1>,<at 1 1>]}
      GreenCurve {keys [<at 0 1>,<at 1 1>]}
      RedCurve {keys [<at 0 1>,<at 1 1>]}
     }}
    Scale Mode=2
   )
   &ScaleColor(Scale Alpha=$System.GlobalFlicker, ScaleA=true, ScaleRGB=true) {module=ScaleColor001}
   &DynamicMaterialParameters(
    Index 0 Param 1=0.68
    Module.Index0Float0Write=true
    Module.Index0Float1Write=false
    Module.Index0Float2Write=false
    Module.Index0Float3Write=false
   )
   &ScaleSpriteSize(
    Scale Factor=&Lerp_Vector2(
     A=<1.3333 0.6667>
     B=<0.6667 1.3333>
     Alpha=&FloatFromCurve(FloatCurve=DataInterface() {propertyValues {Curve {keys [<at 0.0301 -0.5>,<at 0.0468 1.2706>,<at 0.0648 0.5>,<at 0.9428 0.5>,<at 0.9512 -0.1198>,<at 0.9582 1.5743>,<at 0.9828 -0.465>]}}})
    )
    Scale Sprite Size Mode=::NewEnumerator2
   )
   /SetVariables(
    Particles.Digit1:NiagaraFloat=&MakeFloatFromInt(Int=/Niagara/DynamicInputs/Divide/Divide_Int(A:NiagaraInt32=$System.SystemKillCount, B:NiagaraInt32=10))
    Particles.Digit2:NiagaraFloat=&MakeFloatFromInt(Int=$System.SystemKillCount)
   ) {isSetParameters=true}
   /SetVariables(
    Particles.OffsetGlitch:NiagaraFloat=&FloatFromCurve(
     FloatCurve=DataInterface() {propertyValues {Curve {keys [<at 0 1 RCIM_Constant>,<at 0.0219 -0.5 RCIM_Constant>,<at 0.0472 0.5632 RCIM_Constant>,<at 0.0704 0 RCIM_Constant>,<at 0.5903 0.1926 RCIM_Constant>,<at 0.5982 -0.259 RCIM_Constant>,<at 0.6032 0 RCIM_Constant>,<at 0.8214 0.0951 RCIM_Constant>,<at 0.8283 0 RCIM_Constant>,<at 0.9734 0.5765 RCIM_Constant>,<at 0.9907 -0.3544 RCIM_Constant>,<at 1 1 RCIM_Constant>]}}}
     Scale Curve=0.2
    )
    Particles.OffsetShift:NiagaraFloat=&FloatFromCurve(FloatCurve=DataInterface() {propertyValues {Curve {keys [
         {arriveTangent=5.9142, interpMode=RCIM_Cubic, leaveTangent=5.9142, tangentMode=RCTM_User, time=0, value=0}
         {arriveTangent=1.2658, interpMode=RCIM_Cubic, leaveTangent=1.2658, time=0.127, value=0.6147}
         {arriveTangent=0.4129, leaveTangent=1, tangentMode=RCTM_User, time=0.79, value=1}
         {leaveTangent=2, time=1, value=1}
        ]}}})
   ) {isSetParameters=true}
   /SetVariables(
    Particles.Offset1:Vector2f=/Niagara/DynamicInputs/Add/Add_Vector2(
     A:Vector2f=&Lerp_Vector2(A=<0.89 0.52>, B=<0.885 0.6>, Alpha=$Particles.OffsetShift)
     B:Vector2f=&Multiply_Vector2D_ByFloat(Float=$Particles.OffsetGlitch, Vector2D=&RandomRangeVector2D(Minimum=<-1 -1>, Maximum=<1 1>))
    )
    Particles.Offset2:Vector2f=&Add_Vector2(
     A:Vector2f=&Lerp_Vector2(A=<0.11 0.48>, B=<0.115 0.4>, Alpha=$Particles.OffsetShift)
     B:Vector2f=&Multiply_Vector2D_ByFloat(Float=$Particles.OffsetGlitch, Vector2D=&RandomRangeVector2D(Minimum=<-1 -1>, Maximum=<1 1>))
    )
   ) {isSetParameters=true}
   /SetVariables(
    Particles.ColorGlow:LinearColor=&MultiplyLinearColor(
     Linear Color=$Particles.Color
     Scale Factor=&MakeVector4_FromLinearColor(LinearColor=<rgba 0.113 0.5748 1 0.2>)
    )
    Particles.ColorMain:LinearColor=&MultiplyLinearColor(
     Linear Color=$Particles.Color
     Scale Factor=&MakeVector4_FromLinearColor(LinearColor=&FNTimeOfDayColor(
       Night Color=<rgba 0.5 0.7903 1 0.0522>
       Morning Color=<rgba 0.5391 1 1 0.0522>
       Day Color=<rgba 0.5391 1 1 0.0522>
       Evening Color=<rgba 0.5391 1 1 0.0522>
      ))
    )
   ) {isSetParameters=true}
  ]
  rendererClasses [&NiagaraSpriteRendererProperties]
  renderers [
   {rendererClass=&NiagaraSpriteRendererProperties}
   {rendererClass=&NiagaraSpriteRendererProperties, rendererIndex=1}
   {rendererClass=&NiagaraSpriteRendererProperties, rendererIndex=2}
   {rendererClass=&NiagaraSpriteRendererProperties, rendererIndex=3}
  ]
 }
 Illumination {
  emitterUpdateScript [
   &EmitterState(Life Cycle Mode=::NewEnumerator1, Loop Behavior=::NewEnumerator1, Loop Duration=1, UseLoopDelay=false)
   &SpawnBurst_Instantaneous(Spawn Count=3, Use Spawn Probability=false, UseLoopCountLimit=true)
  ]
  particleSpawnScript [
   &PartitionParticles(Number Of Partitions=3)
   &InitializeParticle(
    Color=&MakeLinearColorFromVectorAndFloat("Vector (RGB)"=&MakeVectorFromLinearColorRGB(Color=<rgba 0.3 1.2046 2 1>), "Float (Alpha)"=0.4)
    Lifetime=3
    Mesh Scale=&MaskVector(Mask Source=1, Mask Parameter=$Particles.PartitionParticles.Partition, Mask=0, Masked Vector=<0.2 0.2 3>, Passthrough Vector=<0.06 0.06 1>)
    Mesh Scale Mode=::NewEnumerator4
    Position Mode=::NewEnumerator1
    Sprite Size=/Niagara/DynamicInputs/Helpers/MaskVector2d(
     Mask Source:NiagaraInt32=1
     Mask Parameter:NiagaraInt32=$Particles.PartitionParticles.Partition
     Mask:NiagaraInt32=0
     Masked Vector2D:Vector2f=<24 30>
     Passthrough Vector2D:Vector2f=<14 30>
    )
    Sprite Size Mode=::NewEnumerator4
    UsePositionOffset=false
    Write Material Random=true
   )
   &SocketLocation(
    SamplePosOffset=false
    Socket Index=$Particles.PartitionParticles.Partition
    Socket Reader=$System.Socket Reader
    Socket Sampling Mode=::NewEnumerator3
    VelocitySpeedLimit=false
   )
   &DynamicMaterialParameters(
    Index 0 Param 1=1
    Index 0 Param 2=0.7
    Module.Index0Float0Write=true
    Module.Index0Float1Write=true
    Module.Index0Float2Write=true
    Module.Index0Float3Write=true
   )
  ]
  particleUpdateScript [
   &ParticleState()
   &ScaleColor(Scale Alpha=&Multiply_Float(A=$System.FadeIn, B=$System.GlobalFlicker), ScaleA=true, ScaleRGB=true) {module=ScaleColor001}
   /SetVariables(Particles.Alignment:Vector3f=/Niagara/DynamicInputs/Math/SlerpVector(
     Vector A:Vector3f=&NormalizeVector(
      Vector To Normalize=&Subtract_Vector(A=<0 -27.9897 -1.4653>, B=Linked() {linkedVariable {name=Particles.Initial.Position, type=NiagaraPosition}})
      Safe Normalize=true
      Safe Fallback Threshold=0
      Safe Fallback Vector=<0 0 0.001>
     )
     Vector B:Vector3f=<0 -1 0>
     Alpha:NiagaraFloat=$System.FocusFactor
     Scale:NiagaraFloat=1
    )) {isSetParameters=true}
   /SetVariables(
    Particles.MeshOrientation=&OrientMeshToVector(Look At Direction=$Particles.Alignment, Reference Axis Vector=<0 0 1>)
    Particles.SpriteAlignment=$Particles.Alignment
   ) {isSetParameters=true}
   /SetVariables(
    Particles.Mesh Color:LinearColor=&MultiplyLinearColor(
     Linear Color=$Particles.Color
     Scale Factor=&MakeVector4FromVectorAndFloat(
      "Vector (RGB)"=<1 1 1>
      "Float (Alpha)"=&MaskFloat(Mask Source=1, Mask Parameter=$Particles.PartitionParticles.Partition, Mask=0, Masked Float=0.3, Passthrough Float=1)
     )
    )
    Particles.Scale=&Lerp_Vector(A=<0.024 0.024 1.5>, B=$Particles.Initial.Scale, Alpha=$System.FocusFactor)
    Particles.Sprite Color:LinearColor=&MultiplyLinearColor(
     Linear Color=$Particles.Color
     Scale Factor=&MakeVector4FromVectorAndFloat(
      "Vector (RGB)"=<1 1 1>
      "Float (Alpha)"=&MaskFloat(Mask Source=1, Mask Parameter=$Particles.PartitionParticles.Partition, Mask=0, Masked Float=2, Passthrough Float=1)
     )
    )
    Particles.SpriteSize=&Lerp_Vector2(A=<0 0>, B=$Particles.Initial.SpriteSize, Alpha=$System.FocusFactor)
   ) {isSetParameters=true}
   /SetVariables(
    Particles.LaserColor:LinearColor=&MakeCustomLinearColorFromBool(
     Bool=$User.bIsFrontEnd
     True Color=&Multiply_LinearColorByFloat(
      Linear Color=<rgba 0.5739 1.3493 2 1>
      Float=&Multiply_Float(A=&OneMinusFloat(Float=$System.FocusFactor), B=$System.FadeIn)
     )
     False Color=&Multiply_LinearColorByFloat(
      Linear Color=<rgba 0.3652 1.4896 3 1>
      Float=&Multiply_Float(A=&OneMinusFloat(Float=$System.FocusFactor), B=$System.FadeIn)
     )
    )
    Particles.LaserSize:Vector2f=&Lerp_Vector2(
     A=<0.3 7>
     B=&FN_FrontEndVector2Modifier(Front End:Vector2f=<0.6 15>, Gameplay:Vector2f=<0.3 15>)
     Alpha=$System.FadeIn
    )
   ) {isSetParameters=true}
   /SetVariables(
    Particles.ConeSubUV:NiagaraInt32=&RandomRangeInt(Minimum=0, Maximum=4)
    Particles.Flare Size:Vector2f=&Vector2DFromFloat(Value=&MaskFloat(Mask Source=1, Mask Parameter=$Particles.Initial.PartitionParticles.Partition, Mask=0, Masked Float=13.5358, Passthrough Float=8))
    Particles.FlareCamOffset=5
   ) {isSetParameters=true}
  ]
  rendererClasses [&NiagaraSpriteRendererProperties, &NiagaraMeshRendererProperties]
  renderers [
   {rendererClass=&NiagaraSpriteRendererProperties}
   {rendererClass=&NiagaraMeshRendererProperties, rendererIndex=1}
   {rendererClass=&NiagaraSpriteRendererProperties, rendererIndex=2}
   {rendererClass=&NiagaraSpriteRendererProperties, rendererIndex=3}
  ]
 }
 Particle_Circles {
  emitterUpdateScript [
   &EmitterState(
    Delay First Loop Only=true
    Life Cycle Mode=::NewEnumerator1
    Loop Behavior=::NewEnumerator2
    Loop Count=2
    Loop Delay=0.5
    Loop Duration=$System.ReDelay
    UseLoopDelay=true
   )
   /SetVariables(Emitter.Rate=20) {isSetParameters=true}
   &SpawnRate(
    Spawn Probability=0.8
    SpawnRate=&MakeCustomFloatFromBool(
     Bool=&SetBoolByFloatComparison(A=$Emitter.LoopedAge, B=0.2, Comparison Type=::NewEnumerator6)
     True Float=$Emitter.Rate
     False Float=0
    )
    Use Spawn Probability=false
   )
  ]
  particleSpawnScript [
   &InitializeParticle(
    Color=&MakeLinearColorFromVectorAndFloat("Vector (RGB)"=&MakeVectorFromLinearColorRGB(Color=<rgba 0.2623 0.7758 1 1>), "Float (Alpha)"=1)
    Lifetime=&RandomRangeFloat(Minimum=0.35, Maximum=0.45)
    Sprite Rotation Mode=::NewEnumerator2
    Sprite Size Mode=::NewEnumerator1
    Uniform Sprite Size Max=14
    Uniform Sprite Size Min=7
    UsePositionOffset=false
    Write Material Random=true
   )
   &SocketLocation(SamplePosOffset=false, Socket Index=3, Socket Reader=$System.Socket Reader, Socket Sampling Mode=::NewEnumerator3, VelocitySpeedLimit=false)
   &ShapeLocation(MaskSpawnGroup=false, Override Seed=false, Sphere Radius=15, UseOwnerScale=true)
  ]
  particleUpdateScript [
   &ParticleState()
   &ScaleSpriteSize(Uniform Curve Sprite Scale=DataInterface() {propertyValues {Curve {
       defaultValue=0
       keys [
        {arriveTangent=2.4999, interpMode=RCIM_Cubic, leaveTangent=2.4999, tangentMode=RCTM_User, time=0, value=0.2}
        {arriveTangent=0.9394, interpMode=RCIM_Cubic, leaveTangent=0.9394, tangentMode=RCTM_User, time=0.3526, value=0.7871}
        {arriveTangent=0.0224, interpMode=RCIM_Cubic, leaveTangent=0.0224, tangentMode=RCTM_User, time=1, value=1}
       ]
      }}})
   &ScaleColor(
    Linear Color Curve=DataInterface() {propertyValues {
      AlphaCurve {keys [<at 0.0086 0.1477>,<at 0.1376 0.8>,<at 0.5893 0.0875>,<at 1 0>]}
      BlueCurve {keys [<at 0 1>]}
      GreenCurve {keys [<at 0 1>]}
      RedCurve {keys [<at 0 1>]}
     }}
    Scale Mode=2
   )
   &DynamicMaterialParameters(
    Index 0 Param 1=&FloatFromCurve(
     FloatCurve=DataInterface() {propertyValues {Curve {
        defaultValue=1
        keys [
         {arriveTangent=2.0818, interpMode=RCIM_Cubic, leaveTangent=2.0818, tangentMode=RCTM_User, time=0, value=0}
         {arriveTangent=2.1952, interpMode=RCIM_Cubic, leaveTangent=2.1952, tangentMode=RCTM_User, time=1, value=1}
        ]
       }}}
     Scale Curve=&RandomRangeFloat(Minimum=0, Maximum=0.2)
    )
    Module.Index0Float0Write=true
    Module.Index0Float1Write=true
    Module.Index0Float2Write=false
    Module.Index0Float3Write=false
   )
  ]
  rendererClasses [&NiagaraSpriteRendererProperties]
  renderers [{rendererClass=&NiagaraSpriteRendererProperties}]
 }
 Particle_Circles001 {
  emitterUpdateScript [
   &EmitterState(Life Cycle Mode=::NewEnumerator1, Loop Behavior=::NewEnumerator1, Loop Delay=0.5, Loop Duration=3, UseLoopDelay=true)
   &SpawnBurst_Instantaneous(Use Spawn Probability=false, UseLoopCountLimit=false)
   &SpawnBurst_Instantaneous(Spawn Group=1, Use Spawn Probability=false, UseLoopCountLimit=false) {module=SpawnBurst_Instantaneous001}
   &SpawnBurst_Instantaneous(Spawn Group=1, Spawn Time=$System.ReDelay, Use Spawn Probability=false, UseLoopCountLimit=false) {module=SpawnBurst_Instantaneous002}
  ]
  particleSpawnScript [
   /SetVariables(Particles.VisibilityTag=&SpawnGroup()) {isSetParameters=true}
   &InitializeParticle(
    Color=<rgba 1 1 1 0.313>
    Lifetime=&MaskFloatBySpawnGroup(Masked Float=2.5, Passthrough Float=0.5)
    Sprite Size Mode=::NewEnumerator3
    Uniform Sprite Size=15
    UsePositionOffset=false
    Write Material Random=true
   )
   &CameraOffset(Camera Offset Amount=15)
   &SocketLocation(SamplePosOffset=false, Socket Index=3, Socket Reader=$System.Socket Reader, Socket Sampling Mode=::NewEnumerator3, VelocitySpeedLimit=false)
  ]
  particleUpdateScript [
   &ParticleState()
   &ScaleSpriteSize(Uniform Curve Sprite Scale=DataInterface() {propertyValues {Curve {
       defaultValue=0
       keys [
        {arriveTangent=2.4999, interpMode=RCIM_Cubic, leaveTangent=2.4999, tangentMode=RCTM_User, time=0, value=0.4}
        {arriveTangent=0.3797, interpMode=RCIM_Cubic, leaveTangent=0.3797, tangentMode=RCTM_User, time=0.3494, value=0.9227}
        {arriveTangent=0.0224, interpMode=RCIM_Cubic, leaveTangent=0.0224, tangentMode=RCTM_User, time=1, value=1}
       ]
      }}})
   &ScaleColor(Scale Alpha=$System.GlobalFlicker, ScaleA=true, ScaleRGB=true)
   /SetVariables(
    Particles.GlowColor=&MultiplyLinearColor(
     Linear Color=&ColorFromCurve(ColorCurve=DataInterface() {propertyValues {
        AlphaCurve {keys [<at 0.0086 2>,<at 0.0543 0.8>,<at 0.7898 0.2825>,<at 1 0>]}
        BlueCurve {keys [<at 0 1>]}
        GreenCurve {keys [<at 0 1>]}
        RedCurve {keys [<at 0 1>]}
       }})
     Scale Factor=&MakeVector4_FromLinearColor(LinearColor=$Particles.Color)
    )
    Particles.GlowSize=&Lerp_Vector2(
     A=<50 0.4>
     B=<26.802 16.6494>
     Alpha=&FloatFromCurve(FloatCurve=DataInterface() {propertyValues {Curve {keys [
          {arriveTangent=3.0385, interpMode=RCIM_Cubic, leaveTangent=3.0385, tangentMode=RCTM_User, time=0.004, value=0}
          {arriveTangent=4.0757, interpMode=RCIM_Cubic, leaveTangent=4.0757, time=0.0407, value=0.7794}
          <at 0.2494 1 RCIM_Cubic>
         ]}}})
    )
   ) {isSetParameters=true}
   /SetVariables(
    Particles.BDColor:LinearColor=&MultiplyLinearColor(
     Linear Color=$Particles.Color
     Scale Factor=&MakeVector4_FromLinearColor(LinearColor=&ColorFromCurve(ColorCurve=DataInterface() {propertyValues {
         AlphaCurve {keys [
           <at 0 0.092>
           <at 0.0766 1>
           {arriveTangent=-1.5785, interpMode=RCIM_Cubic, leaveTangent=-1.5785, time=0.1553, value=1}
           <at 0.5883 0.1923>
           <at 1 0>
          ]}
         BlueCurve {keys [<at 0.064 1>,<at 0.6727 1>]}
         GreenCurve {keys [<at 0.064 0.8296>,<at 0.6727 0.5654>]}
         RedCurve {keys [<at 0.064 0.3067>,<at 0.6727 0.2074>]}
        }}))
    )
    Particles.BDParams:Vector4f=&MakeVector4(
     X=&FloatFromCurve(FloatCurve=DataInterface() {propertyValues {Curve {keys [
          {arriveTangent=2.862, interpMode=RCIM_Cubic, leaveTangent=2.862, tangentMode=RCTM_User, time=0.0596, value=0}
          {arriveTangent=1.3064, interpMode=RCIM_Cubic, leaveTangent=1.3064, tangentMode=RCTM_User, time=0.373, value=0.5324}
          {interpMode=RCIM_Cubic, tangentMode=RCTM_User, time=1, value=1}
         ]}}})
     Y=0.3
     Z=0
     W=0
    )
    Particles.BDScale:Vector2f=<33.0535 28>
   ) {isSetParameters=true}
   /SetVariables(
    Particles.BDColor2:LinearColor=&MultiplyLinearColor(
     Linear Color=$Particles.Color
     Scale Factor=&MakeVector4FromVectorAndFloat(
      "Vector (RGB)"=&MakeVectorFromLinearColorRGB(Color=<rgba 0.305 0.8308 1 0.149>)
      "Float (Alpha)"=&Multiply_Float(
       A=&FloatFromCurve(
        FloatCurve=DataInterface() {propertyValues {Curve {keys [
            {arriveTangent=1.2115, interpMode=RCIM_Cubic, time=0, value=0}
            <at 0.5 1 RCIM_Cubic>
            {interpMode=RCIM_Cubic, leaveTangent=0.8059, time=1, value=0}
           ]}}}
        CurveIndex=&Frac_Float(Float=&Multiply_Float(A=$Particles.NormalizedAge, B=1.5))
        Scale Curve=0.15
       )
       B=$System.GlobalFlicker
      )
     )
    )
    Particles.BDParams2:Vector4f=&MakeVector4(
     X=&FloatFromCurve(
      FloatCurve=DataInterface() {propertyValues {Curve {keys [
          {arriveTangent=1.2115, interpMode=RCIM_Cubic, leaveTangent=1.2115, tangentMode=RCTM_User, time=0, value=0.3765}
          {arriveTangent=0.8059, interpMode=RCIM_Cubic, leaveTangent=0.8059, tangentMode=RCTM_User, time=1, value=1}
         ]}}}
      CurveIndex=&Frac_Float(Float=&Multiply_Float(A=$Particles.NormalizedAge, B=1.5))
      Scale Curve=0.7
     )
     Y=0.3
     Z=0
     W=0
    )
   ) {isSetParameters=true}
  ]
  rendererClasses [&NiagaraSpriteRendererProperties]
  renderers [
   {rendererClass=&NiagaraSpriteRendererProperties}
   {rendererClass=&NiagaraSpriteRendererProperties, rendererIndex=1}
   {rendererClass=&NiagaraSpriteRendererProperties, rendererIndex=2}
  ]
 }
 Squares {
  emitterUpdateScript [
   &EmitterState(Life Cycle Mode=::NewEnumerator1, Loop Behavior=::NewEnumerator1, Loop Delay=0.5, Loop Duration=4, UseLoopDelay=false)
   /SetVariables(Emitter.Rate=12) {isSetParameters=true}
   &SpawnRate(
    Spawn Probability=0.8
    SpawnRate=&FloatFromCurve(
     FloatCurve=DataInterface() {propertyValues {Curve {keys [<at 0 0 RCIM_Constant>,<at 0.5 2 RCIM_Constant>,<at 0.75 1 RCIM_Constant>,<at 1 0.2 RCIM_Constant>,<at 2.8 1 RCIM_Constant>,<at 3.2 0 RCIM_Constant>]}}}
     CurveIndex=$Emitter.Age
     Scale Curve=$Emitter.Rate
    )
    Use Spawn Probability=false
   ) {module=SpawnRate001}
   /SetVariables(Emitter.LocalFlicker:NiagaraFloat=&RandomRangeFloat(Minimum=0.4, Maximum=1, Evaluation Type=::NewEnumerator1)) {isSetParameters=true}
  ]
  particleSpawnScript [
   &InitializeParticle(
    Color Maximum=<rgba 0.2001 0.543 1 0.2957>
    Color Minimum=<rgba 0.3154 0.8741 1 0.8>
    Color Mode=::NewEnumerator2
    Lifetime Max=0.6
    Lifetime Min=0.3
    Lifetime Mode=::NewEnumerator1
    Sprite Size Mode=::NewEnumerator1
    Uniform Sprite Size Max=9
    Uniform Sprite Size Min=4
    UsePositionOffset=false
    Write Material Random=true
   )
   &SocketLocation(SamplePosOffset=false, Socket Index=3, Socket Reader=$System.Socket Reader, Socket Sampling Mode=::NewEnumerator3, VelocitySpeedLimit=false)
   &ShapeLocation(MaskSpawnGroup=false, Override Seed=false, Sphere Radius=15, UseOwnerScale=true)
   &DynamicMaterialParameters(
    Index 0 Param 2=&RandomRangeFloat(Minimum=1, Maximum=3)
    Index 0 Param 4=1
    Module.Index0Float0Write=false
    Module.Index0Float1Write=true
    Module.Index0Float2Write=false
    Module.Index0Float3Write=true
   ) {module=DynamicMaterialParameters001}
   &SubUVAnimation(
    End Frame Range Override=2
    Evaluation Type=::NewEnumerator0
    "Sprite Renderer "=DataInterface() {propertyValues {SpriteRenderer=&self:Squares_0.NiagaraSpriteRendererProperties_1}}
    SubUV Animation Mode:ENiagara_SubUVLookupModeV2=::NewEnumerator2
    UseEndFrame=true
    UseStartFrame=true
   ) {module=SubUVAnimation002}
  ]
  particleUpdateScript [
   &ParticleState()
   &ScaleColor(
    Scale Alpha=&Multiply_Float(
     A=$Emitter.LocalFlicker
     B=&FloatFromCurve(
      FloatCurve=DataInterface() {propertyValues {Curve {keys [<at 0 0>,<at 0.4973 1>,<at 1 0>]}}}
      CurveIndex=&Frac_Float(Float=&Add_Float(
        A=&Multiply_Float(A=$Emitter.Age, B=4)
        B=&RandomRangeFloat(Minimum=0, Maximum=1)
       ))
     )
    )
    ScaleA=true
    ScaleRGB=true
   )
   &ScaleColor(
    Linear Color Curve=DataInterface() {propertyValues {
      AlphaCurve {keys [<at 0 0>,<at 0.0533 1>,<at 1 0>]}
      BlueCurve {keys [<at 0 1>,<at 1 0.6957>]}
      GreenCurve {keys [<at 0 1>,<at 1 0.2986>]}
      RedCurve {keys [<at 0 1>,<at 1 0>]}
     }}
    Scale Mode=2
   ) {module=ScaleColor002}
   &ScaleSpriteSize(
    Scale Sprite Size Mode=::NewEnumerator0
    Uniform Scale Factor=&FloatFromCurve(
     FloatCurve=DataInterface() {propertyValues {Curve {keys [<at 0 0.3>,<at 0.0417 0.5529 RCIM_Constant>,<at 0.045 1 RCIM_Constant>]}}}
     Scale Curve=&RandomRangeFloat(Minimum=0.9, Maximum=1, Evaluation Type=::NewEnumerator1)
    )
   )
  ]
  rendererClasses [&NiagaraSpriteRendererProperties]
  renderers [{rendererClass=&NiagaraSpriteRendererProperties}]
 }
}
summary {
 systemName=NS_Backpack_VoidSpray_Score
 emitters [
  {emitterName=Illumination, rendererClasses [&NiagaraSpriteRendererProperties, &NiagaraMeshRendererProperties]}
  {emitterName=Particle_Circles, rendererClasses [&NiagaraSpriteRendererProperties]}
  {emitterName=Particle_Circles001, rendererClasses [&NiagaraSpriteRendererProperties]}
  {emitterName=Emissive_Heart001, rendererClasses [&NiagaraSpriteRendererProperties]}
  {emitterName=Squares, rendererClasses [&NiagaraSpriteRendererProperties]}
 ]
 userVariables [
  User.bIsFrontEnd {type=NiagaraBool}
  User.bIsFrontEndPreview {type=NiagaraBool}
  User.DebugReactivity {type=NiagaraBool}
  User.KillCount {type=NiagaraInt32, defaultValue=67}
 ]
}
