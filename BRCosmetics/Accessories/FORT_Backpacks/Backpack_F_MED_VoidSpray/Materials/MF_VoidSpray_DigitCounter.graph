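# Material function for the VoidSpray backpack that draws a two-digit counter (0-99) and a reactive glow into base color and emissive.
# Tens and ones index a FlipBook digit atlas, with separately offset UVs split by a step mask on U.
# Time since the Timestamp input, over a set duration, dips the digits out mid-reaction while a flickering sphere-mask projection glow peaks.
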MaterialFunction /BRCosmetics/Accessories/FORT_Backpacks/Backpack_F_MED_VoidSpray/Materials/MF_VoidSpray_DigitCounter
graph {nodes [
  n1 = GetMaterialAttributes(None=FunctionInput() {inputName=MA, inputType=FunctionInput_MaterialAttributes}) {attributeGetTypes [#1, #2]}
  n2 = LinearInterpolate(A=NamedRerouteUsage(), B=NamedRerouteUsage(), Alpha=NamedRerouteUsage())
  n3 = NamedRerouteUsage()
  n4 = NamedRerouteUsage()
  n5 = NamedRerouteUsage()
  n6 = NamedRerouteUsage()
  FunctionOutput(None=SetMaterialAttributes!(MaterialAttributes=@n1.MaterialAttributes, EmissiveColor=Add(A=@n2, B=@n3), BaseColor=LinearInterpolate(A=@n4, B=@n5, Alpha=@n6)) {attributeSetTypes [#1, #2]})
  n7 = MaterialFunctionCall(
   UVs=TextureCoordinate()
   Texture Scale=AppendVector(
    A=ScalarParameter() {defaultValue=0.22, group=Digits, parameterName=DigitScaleX}
    B=ScalarParameter() {defaultValue=0.22, group=Digits, parameterName=DigitScaleY}
   )
  ) {materialFunction=&ScaleUVsByCenter}
  n8 = ScalarParameter() {defaultValue=0.72, group=Digits, parameterName=DigitsA_XOffset}
  n9 = ScalarParameter() {defaultValue=-0.91, group=Digits, parameterName=DigitsA_YOffset}
  n10 = Saturate(None=Add(A=@n7.UVs, B=AppendVector(A=@n8, B=@n9)))
  n11 = ComponentMask(None=@n10) {r=true}
  NamedRerouteDeclaration(None=AppendVector(A=SmoothStep(Value=@n11) {constMin=0.5}, B=ComponentMask(None=@n10) {g=true})) {name=UV_A, nodeColor=<rgba 0.336 0.0019 1 1>}
  n12 = ScalarParameter() {defaultValue=0.54, group=Digits, parameterName=DigitsB_XOffset}
  n13 = ScalarParameter() {defaultValue=-0.91, group=Digits, parameterName=DigitsB_YOffset}
  n14 = Saturate(None=Add(A=@n7.UVs, B=AppendVector(A=@n12, B=@n13)))
  n15 = ComponentMask(None=@n14) {r=true}
  NamedRerouteDeclaration(None=AppendVector(A=SmoothStep(Value=@n15) {constMax=0.5}, B=ComponentMask(None=@n14) {g=true})) {name=UV_B, nodeColor=<rgba 1 0 0.2109 1>}
  n16 = TextureCoordinate()
  n17 = AppendVector(
   A=ScalarParameter() {defaultValue=0.33, group=Reactivity, parameterName=GlowOffsetX}
   B=ScalarParameter() {group=Reactivity, parameterName=GlowOffsetY}
  )
  n18 = ScalarParameter() {defaultValue=0.2, group=Reactivity, parameterName=ProjectionGlowRadius}
  n19 = ScalarParameter() {defaultValue=0.05, group=Reactivity, parameterName=ProjectionGlowHardness}
  n20 = Multiply(A=Multiply(A=SphereMask(A=@n16, B=@n17, Radius=@n18, Hardness=@n19), B=NamedRerouteUsage()), B=NamedRerouteUsage())
  n21 = NamedRerouteUsage()
  n22 = Time()
  n23 = ScalarParameter() {defaultValue=7, group=Reactivity, parameterName=FlickeringGlowSpeed}
  n24 = Divide(A=Sine(None=Multiply(A=@n22, B=@n23))) {constB=0.85}
  n25 = Multiply(
   A=ScalarParameter() {defaultValue=5, group=Reactivity, parameterName=ProjectionGlowStrength}
   B=LinearInterpolate(Alpha=Saturate(None=@n24)) {constA=0.7}
  )
  NamedRerouteDeclaration(None=Multiply(A=Saturate(None=@n20), B=Multiply(A=@n21, B=@n25) {constB=3})) {name=ProjectionGlow, nodeColor=<rgba 0 0.8455 1 1>}
  NamedRerouteDeclaration(None=FunctionInput() {inputName=EmissiveMask, inputType=FunctionInput_Scalar}) {name=EmissiveChannelMask, nodeColor=<rgba 0 1 0 1>}
  n26 = Divide(A=NamedRerouteUsage()) {constB=10}
  n27 = Saturate(None=Multiply(A=Floor(None=@n26)) {constB=0.1})
  n28 = Fmod(A=NamedRerouteUsage(), B=Constant() {r=10})
  n29 = Saturate(None=Multiply(A=Floor(None=@n28)) {constB=0.1})
  n30 = NamedRerouteUsage()
  n31 = NamedRerouteUsage()
  n32 = NamedRerouteUsage()
  n33 = NamedRerouteUsage()
  NamedRerouteDeclaration(None=MaterialFunctionCall(
    "Animation  Phase (0-1)"=LinearInterpolate(A=@n27, B=@n29, Alpha=@n30)
    Number of Rows=ScalarParameter() {defaultValue=2, parameterName=Rows}
    Number of Columns=ScalarParameter() {defaultValue=5, parameterName=Columns}
    Texture=NamedRerouteUsage()
    UVs=LinearInterpolate(A=@n31, B=@n32, Alpha=@n33)
    "Use Mip Bias (T) Level (F)"=StaticBoolParameter() {group=Digits, parameterName="Use Mip Bias (T) Level (F) (SB)"}
   ) {materialFunction=&FlipBook}.Alpha) {name=Digits, nodeColor=<rgba 1 0 0.7734 1>}
  NamedRerouteDeclaration(None=Clamp(None=FunctionInput() {inputName=Digit, inputType=FunctionInput_Scalar}) {maxDefault=99}) {name=CurrentNumber, nodeColor=<rgba 0 0.0187 1 1>}
  n34 = TextureObject() {
   samplerType=SAMPLERTYPE_LinearColor
   texture=/BRCosmetics/Accessories/FORT_Backpacks/Backpack_F_MED_VoidSpray/Textures/T_VoidSpray_Backpack_Digits
  }
  NamedRerouteDeclaration(None=@n34) {name=DigitTexture, nodeColor=<rgba 0 0.8377 1 1>}
  NamedRerouteDeclaration(None=VectorParameter() {group=Digits, parameterName=DigitColor, defaultValue=<rgba 0 0.1191 0.4844 1>}.RGB) {name=Digit Color, nodeColor=<rgba 0 0.9766 1 1>}
  NamedRerouteDeclaration(None=@n1.EmissiveColor) {name=Emissive, nodeColor=<rgba 0 0.0156 1 1>}
  NamedRerouteDeclaration(None=@n1.BaseColor) {name=BaseColor, nodeColor=<rgba 0.3359 1 0 1>}
  n35 = NamedRerouteUsage()
  n36 = NamedRerouteUsage()
  n37 = Subtract(A=Time(), B=FunctionInput() {inputName=Timestamp, inputType=FunctionInput_Scalar})
  n38 = ScalarParameter() {defaultValue=3, group=Reactivity, parameterName=ReactivityDuration}
  n39 = Divide(A=@n37, B=@n38)
  n40 = Power(
   Base=Abs(None=ConstantBiasScale(None=@n39) {bias=-0.5, scale=2})
   Exp=ScalarParameter() {defaultValue=5, group=Reactivity, parameterName=DigitFadePower}
  ) {constExponent=0.1}
  NamedRerouteDeclaration(None=Multiply(A=Multiply(A=@n35, B=@n36), B=Saturate(None=@n40))) {name=DigitMask, nodeColor=<rgba 0.7109 1 0 1>}
  n41 = ScalarParameter() {defaultValue=0.317, group=Digits, parameterName=DigitLerpMask}
  n42 = ComponentMask(None=TextureCoordinate()) {r=true}
  NamedRerouteDeclaration(None=Saturate(None=Step(Y=@n41, X=@n42) {constY=0.317})) {name=LerpMask, nodeColor=<rgba 0 1 0.6016 1>}
  n43 = Divide(A=@n37, B=@n38)
  n44 = OneMinus(None=Abs(None=ConstantBiasScale(None=@n43) {bias=-0.5, scale=2}))
  n45 = ScalarParameter() {defaultValue=0.5, group=Reactivity, parameterName=GlowFadePower}
  NamedRerouteDeclaration(None=Saturate(None=Power(Base=@n44, Exp=@n45) {constExponent=0.8})) {name=GlowMask, nodeColor=<rgba 0.4531 0 1 1>}
 ]}
