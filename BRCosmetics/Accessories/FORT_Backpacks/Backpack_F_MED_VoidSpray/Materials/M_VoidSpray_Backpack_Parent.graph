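# Parent material for the Void Spray backpack cosmetic, adding a reactive digit counter driven by kill count to the standard character shading.
# MF_VoidSpray_DigitCounter sits between the 2023 character parent PreFX and PostFX functions, modifying the material attributes.
# The counter takes KillCount as its digit, a Timestamp parameter for reactivity, and an emissive mask.
# The green channel of the PreFX Tech Art Mask is exposed as the EmissiveChannelMask reroute.
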Material /BRCosmetics/Accessories/FORT_Backpacks/Backpack_F_MED_VoidSpray/Materials/M_VoidSpray_Backpack_Parent
graph {
 properties {
  bAutomaticallySetUsageInEditor=false
  bHasPixelAnimation=true
  bUseMaterialAttributes=true
  bUsedWithMeshDeformer=true
  bUsedWithMorphTargets=true
  bUsedWithSkeletalMesh=true
  pixelDepthOffsetMode=PDOM_Legacy
  shadingModel=MSM_FromMaterialExpression
  usedShadingModels=Default Lit | Subsurface | Clear Coat | Subsurface Profile | Cloth
 }
 nodes [
  n1 = MaterialFunctionCall() {materialFunction=&MF_CharacterParent_2023_PreFX}
  MaterialFunctionCall(
   Material=MaterialFunctionCall!(
    MA=@n1.Result
    EmissiveMask=NamedRerouteUsage()
    Timestamp=ScalarParameter() {defaultValue=-10000, group=Reactivity, parameterName=Timestamp}
    Digit=ScalarParameter() {parameterName=KillCount}
   ) {materialFunction=&/MF_VoidSpray_DigitCounter}.Result
   M=@n1.M
   Pre Skinned Local Position=@n1.M
   Distance Blend=@n1.M
   Fresnel=@n1.M
  ) {materialFunction=&MF_CharacterParent_2023_PostFX}
  NamedRerouteDeclaration(None=ComponentMask(None=@n1.Tech Art Mask) {g=true}) {name=EmissiveChannelMask, nodeColor=<rgba 0 1 0 1>}
 ]
}
